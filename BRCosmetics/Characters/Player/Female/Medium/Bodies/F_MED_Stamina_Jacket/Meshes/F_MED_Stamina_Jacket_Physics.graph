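# Physics asset for the female medium Stamina Jacket cosmetic body, setting up the coat's dynamic bones for physics.
# Eight three-segment capsule chains form the coat, joined by swing- and tight twist-limited constraints; only back chains attach to small parent spheres.
# A few angular-free cross-constraints tie neighbouring chains together.
# Pelvis, lower spine, leg and foot capsules provide body shapes, plus a huge root capsule mostly below the origin, apparently a floor.
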PhysicsAsset /BRCosmetics/Characters/Player/Female/Medium/Bodies/F_MED_Stamina_Jacket/Meshes/F_MED_Stamina_Jacket_Physics
bodies {
 calf_l [{
   shapeType=Capsule
   length=46.8766
   radius=8.1047
   shapeName=calf_l_capsule
   center=<-23.5125 -0.3931 0>
   rotation=<rot 90 -179.7016 -177.3881>
  }]
 calf_r [{
   shapeType=Capsule
   length=40.808
   radius=7.3844
   shapeName=calf_r_capsule
   center=<19.7478 1.3469 0.6414>
   rotation=<rot 88.7172 0.0001 0.0002>
  }]
 dyn_coat_01_bk_l [{shapeType=Capsule, length=8.3401, radius=12.3904, shapeName=dyn_coat_01_bk_l_capsule, center=<12 0 -10>, rotation=<rot 0 0 90>}]
 dyn_coat_01_bk_r [{shapeType=Capsule, length=10.0147, radius=11.1706, shapeName=dyn_coat_01_bk_r_capsule, center=<12 0 -7.5>, rotation=<rot 0 0 90>}]
 dyn_coat_01_bk_side_l [{shapeType=Capsule, length=13.4616, radius=12.6746, shapeName=dyn_coat_01_bk_side_l_capsule, center=<11.5915 0 -12.9>, rotation=<rot 0 0 90>}]
 dyn_coat_01_bk_side_r [{shapeType=Capsule, length=10.0138, radius=10.848, shapeName=dyn_coat_01_bk_side_r_capsule, center=<10 0 -10.4>, rotation=<rot 0 0 90>}]
 dyn_coat_01_frnt_l [{shapeType=Capsule, length=9.4534, radius=10.8164, shapeName=dyn_coat_01_frnt_l_capsule, center=<14.3682 0 -9.35>, rotation=<rot 0 0 90>}]
 dyn_coat_01_frnt_r [{shapeType=Capsule, length=6.4033, radius=10.8164, shapeName=dyn_coat_01_frnt_r_capsule, center=<12.8293 0 -9.8>, rotation=<rot 0 0 90>}]
 dyn_coat_01_frnt_side_l [{
   shapeType=Capsule
   length=8.2008
   radius=13.2774
   shapeName=dyn_coat_01_frnt_side_l_capsule
   center=<12.4794 0 -12.76>
   rotation=<rot 0 0 90>
  }]
 dyn_coat_01_frnt_side_r [{shapeType=Capsule, length=14.1545, radius=10.8164, shapeName=dyn_coat_01_frnt_side_r_capsule, center=<11 0 -10.32>, rotation=<rot 0 0 90>}]
 dyn_coat_02_bk_l [{shapeType=Capsule, length=12.1432, radius=10.8151, shapeName=dyn_coat_02_bk_l_capsule, center=<10.11 0 -9>, rotation=<rot 0 0 90>}]
 dyn_coat_02_bk_r [{shapeType=Capsule, length=10.683, radius=11.5137, shapeName=dyn_coat_02_bk_r_capsule, center=<10.11 0 -7>, rotation=<rot 0 0 90>}]
 dyn_coat_02_bk_side_l [{
   shapeType=Capsule
   length=15.9726
   radius=10.3889
   shapeName=dyn_coat_02_bk_side_l_capsule
   center=<10.11 0 -7.95>
   rotation=<rot 0 2.9664 90>
  }]
 dyn_coat_02_bk_side_r [{shapeType=Capsule, length=5.4975, radius=10.2922, shapeName=dyn_coat_02_bk_side_r_capsule, center=<10.11 0 -7.8>, rotation=<rot 0 0 90>}]
 dyn_coat_02_frnt_l [{shapeType=Capsule, length=9.9353, radius=10.4819, shapeName=dyn_coat_02_frnt_l_capsule, center=<12.3489 0 -7.64>, rotation=<rot 0 0 90>}]
 dyn_coat_02_frnt_r [{shapeType=Capsule, length=9.6527, radius=10.5133, shapeName=dyn_coat_02_frnt_r_capsule, center=<10.11 0 -8.75>, rotation=<rot 0 0 90>}]
 dyn_coat_02_frnt_side_l [{
   shapeType=Capsule
   length=11.4642
   radius=10.4252
   shapeName=dyn_coat_02_frnt_side_l_capsule
   center=<12.5619 0 -9.5884>
   rotation=<rot 0 0 90>
  }]
 dyn_coat_02_frnt_side_r [{shapeType=Capsule, length=15.2556, radius=10.2999, shapeName=dyn_coat_02_frnt_side_r_capsule, center=<10 0 -9.8>, rotation=<rot 0 0 90>}]
 dyn_coat_03_bk_l [{shapeType=Capsule, length=32.8722, radius=9.0197, shapeName=dyn_coat_03_bk_l_capsule, center=<8 0 -4.3>, rotation=<rot 0.0001 0 79.4682>}]
 dyn_coat_03_bk_r [{shapeType=Capsule, length=34.9583, radius=8.5687, shapeName=dyn_coat_03_bk_r_capsule, center=<9.3 0 -3.44>, rotation=<rot 0 0 98.91>}]
 dyn_coat_03_bk_side_l [{shapeType=Capsule, length=34.9957, radius=6.9775, shapeName=dyn_coat_03_bk_side_l_capsule, center=<9.4474 0 -4>, rotation=<rot 0 4.9864 90>}]
 dyn_coat_03_bk_side_r [{shapeType=Capsule, length=34.1726, radius=8.8937, shapeName=dyn_coat_03_bk_side_r_capsule, center=<7.8 0 -7.95>, rotation=<rot 0 0 90>}]
 dyn_coat_03_frnt_l [{shapeType=Capsule, length=12.347, radius=8.9371, shapeName=dyn_coat_03_frnt_l_capsule, center=<10.476 0 -5.9>, rotation=<rot 0 0 90>}]
 dyn_coat_03_frnt_r [{shapeType=Capsule, length=13.525, radius=10.7863, shapeName=dyn_coat_03_frnt_r_capsule, center=<9.1283 0 -7.5>, rotation=<rot 0 0 90>}]
 dyn_coat_03_frnt_side_l [{
   shapeType=Capsule
   length=13.2689
   radius=9.0264
   shapeName=dyn_coat_03_frnt_side_l_capsule
   center=<10 0.4835 -8.3>
   rotation=<rot 0 0 79.7824>
  }]
 dyn_coat_03_frnt_side_r [{shapeType=Capsule, length=13.957, radius=9.163, shapeName=dyn_coat_03_frnt_side_r_capsule, center=<8.4286 0 -7.87>, rotation=<rot 0 0 90>}]
 dyn_coat_parent_bk_l [{shapeType=Sphere, radius=0.5, shapeName=dyn_coat_parent_bk_l_sphere}]
 dyn_coat_parent_bk_r [{shapeType=Sphere, radius=0.5, shapeName=dyn_coat_parent_bk_r_sphere}]
 dyn_coat_parent_bk_side_l [{shapeType=Sphere, radius=0.5, shapeName=dyn_coat_parent_bk_side_l_sphere}]
 dyn_coat_parent_bk_side_r [{shapeType=Sphere, radius=0.5, shapeName=dyn_coat_parent_bk_side_r_sphere}]
 foot_l [{
   shapeType=Capsule
   length=19.1388
   radius=11.8665
   shapeName=foot_l_capsule
   center=<-3.5986 5.9956 1.5373>
   rotation=<rot 0 -0.0001 -90.8661>
  }]
 foot_r [{shapeType=Capsule, length=19.1388, radius=16.8161, shapeName=foot_r_capsule, center=<0 -5.7817 0.0245>, rotation=<rot 0 -0.0001 -95.2233>}]
 pelvis [{shapeType=Capsule, length=9.7578, radius=12.0325, shapeName=pelvis_capsule, center=<-8.3686 -2.3378 -0.3565>}]
 root [{shapeType=Capsule, length=2.75, radius=455, center=<-0.0002 -39.8738 -443.6007>}]
 spine_01 [{shapeType=Capsule, length=28.731, radius=10.7752, shapeName=spine_01_capsule, center=<4.6312 0 0>, rotation=<rot 90 0 0>}]
 thigh_l [{
   shapeType=Capsule
   length=37.8401
   radius=8.243
   shapeName=thigh_r_capsule
   center=<-21.6689 -0.8736 1.6099>
   rotation=<rot 89.4559 179.9984 176.3576>
  }]
 thigh_r [{
   shapeType=Capsule
   length=39.3837
   radius=7.4481
   shapeName=thigh_r_capsule
   center=<27.5655 0.1677 -0.3459>
   rotation=<rot 89.7075 0.0006 0.0007>
  }]
}
constraints [
 {bone1Name=dyn_coat_02_bk_r, bone2Name=dyn_coat_01_bk_r, swing1LimitDegrees=30, swing2LimitDegrees=35, twistLimitDegrees=2}
 {bone1Name=dyn_coat_02_bk_l, bone2Name=dyn_coat_01_bk_l, swing1LimitDegrees=20, swing2LimitDegrees=35, twistLimitDegrees=2}
 {bone1Name=dyn_coat_02_bk_side_r, bone2Name=dyn_coat_01_bk_side_r, swing1LimitDegrees=30, swing2LimitDegrees=35, twistLimitDegrees=2}
 {bone1Name=dyn_coat_02_bk_side_l, bone2Name=dyn_coat_01_bk_side_l, swing1LimitDegrees=30, swing2LimitDegrees=35, twistLimitDegrees=2}
 {bone1Name=dyn_coat_02_frnt_side_r, bone2Name=dyn_coat_01_frnt_side_r, swing1LimitDegrees=30, swing2LimitDegrees=35, twistLimitDegrees=2}
 {bone1Name=dyn_coat_02_frnt_side_l, bone2Name=dyn_coat_01_frnt_side_l, swing1LimitDegrees=30, swing2LimitDegrees=35, twistLimitDegrees=2}
 {bone1Name=dyn_coat_02_frnt_r, bone2Name=dyn_coat_01_frnt_r, swing1LimitDegrees=30, swing2LimitDegrees=35, twistLimitDegrees=2}
 {bone1Name=dyn_coat_02_frnt_l, bone2Name=dyn_coat_01_frnt_l, swing1LimitDegrees=30, swing2LimitDegrees=35, twistLimitDegrees=2}
 {bone1Name=dyn_coat_03_bk_r, bone2Name=dyn_coat_02_bk_r, swing1LimitDegrees=30, swing2LimitDegrees=35, twistLimitDegrees=2}
 {bone1Name=dyn_coat_03_frnt_l, bone2Name=dyn_coat_02_frnt_l, swing1LimitDegrees=30, swing2LimitDegrees=45, twistLimitDegrees=2}
 {bone1Name=dyn_coat_03_frnt_r, bone2Name=dyn_coat_02_frnt_r, swing1LimitDegrees=30, swing2LimitDegrees=35, twistLimitDegrees=2}
 {bone1Name=dyn_coat_03_frnt_side_l, bone2Name=dyn_coat_02_frnt_side_l, swing1LimitDegrees=30, swing2LimitDegrees=35, twistLimitDegrees=2}
 {bone1Name=dyn_coat_03_frnt_side_r, bone2Name=dyn_coat_02_frnt_side_r, swing1LimitDegrees=30, swing2LimitDegrees=35, twistLimitDegrees=2}
 {bone1Name=dyn_coat_03_bk_side_l, bone2Name=dyn_coat_02_bk_side_l, swing1LimitDegrees=30, swing2LimitDegrees=35, twistLimitDegrees=2}
 {bone1Name=dyn_coat_03_bk_side_r, bone2Name=dyn_coat_02_bk_side_r, swing1LimitDegrees=30, swing2LimitDegrees=35, twistLimitDegrees=2}
 {bone1Name=dyn_coat_03_bk_l, bone2Name=dyn_coat_02_bk_l, swing1LimitDegrees=20, swing2LimitDegrees=35, twistLimitDegrees=2}
 {bone1Name=dyn_coat_01_frnt_l, bone2Name=dyn_coat_01_frnt_side_l, swing1LimitDegrees=30, swing2LimitDegrees=35, twistLimitDegrees=2}
 {bone1Name=dyn_coat_01_bk_l, bone2Name=dyn_coat_parent_bk_l, swing1LimitDegrees=20, swing2LimitDegrees=35, twistLimitDegrees=2}
 {bone1Name=dyn_coat_01_bk_r, bone2Name=dyn_coat_parent_bk_r, swing1LimitDegrees=30, swing2LimitDegrees=35, twistLimitDegrees=2}
 {bone1Name=dyn_coat_01_bk_side_l, bone2Name=dyn_coat_parent_bk_side_l, swing1LimitDegrees=30, swing2LimitDegrees=35, twistLimitDegrees=2}
 {bone1Name=dyn_coat_01_bk_side_r, bone2Name=dyn_coat_parent_bk_side_r, swing1LimitDegrees=35, swing2LimitDegrees=30, twistLimitDegrees=3}
 {
  bone1Name=dyn_coat_03_bk_r
  bone2Name=dyn_coat_03_bk_side_r
  swing1LimitDegrees=45
  swing1Motion=Free
  swing2LimitDegrees=45
  swing2Motion=Free
  twistLimitDegrees=45
  twistMotion=Free
 }
 {
  bone1Name=dyn_coat_02_frnt_side_l
  bone2Name=dyn_coat_02_bk_side_l
  swing1LimitDegrees=45
  swing1Motion=Free
  swing2LimitDegrees=45
  swing2Motion=Free
  twistLimitDegrees=45
  twistMotion=Free
 }
 {
  bone1Name=dyn_coat_01_bk_side_l
  bone2Name=dyn_coat_01_bk_l
  swing1LimitDegrees=45
  swing1Motion=Free
  swing2LimitDegrees=45
  swing2Motion=Free
  twistLimitDegrees=45
  twistMotion=Free
 }
 {
  bone1Name=dyn_coat_01_bk_side_r
  bone2Name=dyn_coat_01_bk_r
  swing1LimitDegrees=45
  swing1Motion=Free
  swing2LimitDegrees=45
  swing2Motion=Free
  twistLimitDegrees=45
  twistMotion=Free
 }
]
