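# Control Rig for the female medium Stamina Jacket body, named for skydiving, that adjusts the coat's dynamic bones as the legs move.
# A sequence runs left and right branches, each using a spherical pose reader on that side's calf as the blend weight.
# That weight scales fixed rotation offsets on back and front coat bones.
# The sides are asymmetric: right back offsets are zero, so only right front bones rotate.
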ControlRigBlueprint /BRCosmetics/Characters/Player/Female/Medium/Bodies/F_MED_Stamina_Jacket/Meshes/F_MED_Stamina_Jacket_Skydive_CtrlRig
graphs [
 RigVMFunctionLibrary {links [], nodes []}
 RigVMModel {nodes [
   RigUnit_BeginExecution = &RigUnit_BeginExecution()
   n1 = /Script/ControlRig.RigUnit_SequenceExecution(ExecuteContext=@RigUnit_BeginExecution.ExecutePin) {is=struct RigVMUnitNode}
   n2 = &RigUnit_SphericalPoseReader(
    ActiveRegionScaleFactors=(PositiveWidth=3.5,NegativeWidth=1,PositiveHeight=3,NegativeHeight=3)
    ActiveRegionSize=0.1
    Debug=(bDrawDebug=false,bDraw2D=false,DebugScale=25,DebugSegments=20,DebugThickness=0.25)
    DriverAxis=(X=-1,Y=0,Z=0)
    DriverItem=(Type=Bone,Name="calf_l")
    FalloffRegionScaleFactors=(PositiveWidth=1,NegativeWidth=1,PositiveHeight=1,NegativeHeight=1)
    FalloffSize=1
    OptionalParentItem=(Type=Bone)
    RotationOffset=(X=-90,Y=0,Z=0)
    ExecutePin=@n1.A
   )
   > &RigUnit_ModifyTransforms(
    ItemToModify=((Item=(Type=Bone,Name="dyn_coat_parent_bk_side_l"),Transform=(Rotation=(X=0,Y=0,Z=0,W=1),Translation=(X=0,Y=0,Z=0),Scale3D=(X=1,Y=1,Z=1))))
    ItemToModify.0.Transform.Rotation=&RigVMFunction_MathQuaternionFromEuler(Euler=(X=0,Y=20,Z=0)).Result
    Weight=@n2.OutputParam
   )
   > &RigUnit_ModifyTransforms(
    ItemToModify=((Item=(Type=Bone,Name="dyn_coat_parent_bk_l"),Transform=(Rotation=(X=0,Y=0,Z=0,W=1),Translation=(X=0,Y=0,Z=0),Scale3D=(X=1,Y=1,Z=1))))
    ItemToModify.0.Transform.Rotation=&RigVMFunction_MathQuaternionFromEuler(Euler=(X=0,Y=15,Z=10)).Result
    Weight=@n2.OutputParam
   )
   > &RigUnit_ModifyTransforms(
    ItemToModify=((Item=(Type=Bone,Name="dyn_coat_01_frnt_side_l"),Transform=(Rotation=(X=0,Y=0,Z=0,W=1),Translation=(X=0,Y=0,Z=0),Scale3D=(X=1,Y=1,Z=1))))
    ItemToModify.0.Transform.Rotation=&RigVMFunction_MathQuaternionFromEuler(Euler=(X=0,Y=-20,Z=0)).Result
    Weight=@n2.OutputParam
   )
   > &RigUnit_ModifyTransforms(
    ItemToModify=((Item=(Type=Bone,Name="dyn_coat_01_frnt_l"),Transform=(Rotation=(X=0,Y=0,Z=0,W=1),Translation=(X=0,Y=0,Z=0),Scale3D=(X=1,Y=1,Z=1))))
    ItemToModify.0.Transform.Rotation=&RigVMFunction_MathQuaternionFromEuler(Euler=(X=0,Y=10,Z=0)).Result
    Weight=@n2.OutputParam
   )
   n3 = &RigUnit_SphericalPoseReader(
    ActiveRegionScaleFactors=(PositiveWidth=2,NegativeWidth=2,PositiveHeight=2,NegativeHeight=2)
    ActiveRegionSize=0.1
    Debug=(bDrawDebug=false,DebugScale=25,DebugSegments=20,DebugThickness=0.25)
    DriverAxis=(X=1,Y=0,Z=0)
    DriverItem=(Type=Bone,Name="calf_r")
    FalloffRegionScaleFactors=(PositiveWidth=1,NegativeWidth=1,PositiveHeight=1,NegativeHeight=1)
    FalloffSize=1
    FlipWidthScaling=false
    OptionalParentItem=(Type=Bone)
    RotationOffset=(X=90,Y=0,Z=0)
    ExecutePin=@n1.B
   )
   > &RigUnit_ModifyTransforms(
    ItemToModify=((Item=(Type=Bone,Name="dyn_coat_parent_bk_side_r"),Transform=(Rotation=(X=0,Y=0,Z=0,W=1),Translation=(X=0,Y=0,Z=0),Scale3D=(X=1,Y=1,Z=1))))
    ItemToModify.0.Transform.Rotation=&RigVMFunction_MathQuaternionFromEuler(Euler=(X=0,Y=0,Z=0)).Result
    Weight=@n3.OutputParam
   )
   > &RigUnit_ModifyTransforms(
    ItemToModify=((Item=(Type=Bone,Name="dyn_coat_parent_bk_r"),Transform=(Rotation=(X=0,Y=0,Z=0,W=1),Translation=(X=0,Y=0,Z=0),Scale3D=(X=1,Y=1,Z=1))))
    ItemToModify.0.Transform.Rotation=&RigVMFunction_MathQuaternionFromEuler(Euler=(X=0,Y=0,Z=0)).Result
    Weight=@n3.OutputParam
   )
   > &RigUnit_ModifyTransforms(
    ItemToModify=((Item=(Type=Bone,Name="dyn_coat_01_frnt_r"),Transform=(Rotation=(X=0,Y=0,Z=0,W=1),Translation=(X=0,Y=0,Z=0),Scale3D=(X=1,Y=1,Z=1))))
    ItemToModify.0.Transform.Rotation=&RigVMFunction_MathQuaternionFromEuler(Euler=(X=0,Y=20,Z=0)).Result
    Weight=@n3.OutputParam
   )
   > &RigUnit_ModifyTransforms(
    ItemToModify=((Item=(Type=Bone,Name="dyn_coat_01_frnt_side_r"),Transform=(Rotation=(X=0,Y=0,Z=0,W=1),Translation=(X=0,Y=0,Z=0),Scale3D=(X=1,Y=1,Z=1))))
    ItemToModify.0.Transform.Rotation=&RigVMFunction_MathQuaternionFromEuler(Euler=(X=0,Y=20,Z=0)).Result
    Weight=@n3.OutputParam
   )
  ]}
]
hierarchy [
 root {loc=<0 0 0>, rot=<0 0 0 1>}
 pelvis {parent=root, loc=<0 -0.5301 98.0571>, rot=<0.0127 0.707 -0.0127 -0.707>}
 spine_01 {parent=pelvis, loc=<6.1208 -0.9328 0>, rot=<0 0 0.0576 -0.9983>}
 spine_02 {parent=spine_01, loc=<6.6022 -0.0153 0>, rot=<0 0 0.0022 -1>}
 spine_03 {parent=spine_02, loc=<6.4919 -0.0105 0>, rot=<0 0 0.0015 -1>}
 spine_04 {parent=spine_03, loc=<6.9596 -0.007 0>, rot=<0 0 0.0009 -1>}
 spine_05 {parent=spine_04, loc=<6.9981 -0.0017 0>, rot=<0 0 -0.0042 -1>}
 neck_01 {parent=spine_05, loc=<11.3138 0.1529 0>, rot=<0 0 -0.2003 -0.9797>}
 neck_02 {parent=neck_01, loc=<4.5627 0 0>, rot=<0 0 0.0504 -0.9987>}
 head {parent=neck_02, loc=<5.3914 0 0>, rot=<0 0 0.115 -0.9934>}
 clavicle_l {parent=spine_05, loc=<5.4831 -0.1053 -3.2946>, rot=<0.0938 -0.7388 0.0307 -0.6666>}
 upperarm_l {parent=clavicle_l, loc=<9.9486 0 0>, rot=<0.0371 -0.2992 -0.0705 -0.9509>}
 lowerarm_l {parent=upperarm_l, loc=<26.4793 0 0>, rot=<0 0 -0.2382 -0.9712>}
 hand_l {parent=lowerarm_l, loc=<22.7458 0 0>, rot=<-0.6747 -0.0182 0.0605 -0.7354>}
 thumb_01_l {parent=hand_l, loc=<1.3118 -0.3191 -2.459>, rot=<0.6546 -0.4042 0.0664 -0.6354>}
 thumb_02_l {parent=thumb_01_l, loc=<5.9322 -0.0253 0.1416>, rot=<0.0121 0.0046 0.2964 -0.955>}
 thumb_03_l {parent=thumb_02_l, loc=<2.2223 0.3682 0>, rot=<0 0 0.0172 -0.9999>}
 index_metacarpal_l {parent=hand_l, loc=<1.6535 0.0183 -2.96>, rot=<0.0158 -0.0124 -0.0173 -0.9996>}
 index_01_l {parent=index_metacarpal_l, loc=<8.3371 0.1718 -0.2461>, rot=<0.0363 0.0499 0.142 -0.9879>}
 index_02_l {parent=index_01_l, loc=<3.4919 -0.3004 0>, rot=<0 0 0.2948 -0.9556>}
 index_03_l {parent=index_02_l, loc=<2.5578 -0.123 0>, rot=<0 0 0.1968 -0.9804>}
 middle_metacarpal_l {parent=hand_l, loc=<2.1856 -0.1725 -0.5984>, rot=<0.006 0.0168 -0.0329 -0.9993>}
 middle_01_l {parent=middle_metacarpal_l, loc=<7.6374 0.346 -0.1508>, rot=<-0.0085 0.0328 0.1813 -0.9829>}
 middle_02_l {parent=middle_01_l, loc=<3.9108 0 0>, rot=<0 0 0.2808 -0.9598>}
 middle_03_l {parent=middle_02_l, loc=<2.6103 0.0517 0>, rot=<0 0 0.1446 -0.9895>}
 ring_metacarpal_l {parent=hand_l, loc=<2.2074 -0.3625 1.3638>, rot=<-0.0284 0.0722 -0.0458 -0.9959>}
 ring_01_l {parent=ring_metacarpal_l, loc=<7.2483 0.0329 -0.1426>, rot=<-0.0456 -0.0277 0.2951 -0.954>}
 ring_02_l {parent=ring_01_l, loc=<3.7208 0.4647 0>, rot=<0 0 0.1583 -0.9874>}
 ring_03_l {parent=ring_02_l, loc=<2.4152 0.114 0>, rot=<0 0 0.1504 -0.9886>}
 pinky_metacarpal_l {parent=hand_l, loc=<1.7616 -0.9753 3.3814>, rot=<-0.0397 0.0814 -0.069 -0.9935>}
 pinky_01_l {parent=pinky_metacarpal_l, loc=<7.1933 -0.4205 -0.2601>, rot=<-0.0487 -0.0063 0.2556 -0.9655>}
 pinky_02_l {parent=pinky_01_l, loc=<2.8274 0.0893 0>, rot=<0 0 0.2474 -0.9689>}
 pinky_03_l {parent=pinky_02_l, loc=<2.2046 0.0307 0>, rot=<0 0 0.125 -0.9922>}
 lowerarm_twist_02_l {parent=lowerarm_l, loc=<9.6199 0 0>, rot=<0 0 0 -1>}
 lowerarm_twist_01_l {parent=lowerarm_l, loc=<17.8549 0 0>, rot=<0 0 0 -1>}
 upperarm_twist_01_l {parent=upperarm_l, loc=<0.0041 0 0>, rot=<0 0 0 1>}
 upperarm_twist_02_l {parent=upperarm_l, loc=<13.7941 0 0>, rot=<0 0 0 1>}
 clavicle_r {parent=spine_05, loc=<5.4831 -0.1053 3.2946>, rot=<0.7388 0.0938 -0.6666 -0.0307>}
 upperarm_r {parent=clavicle_r, loc=<-9.9486 0 0>, rot=<0.0371 -0.2992 -0.0705 -0.9509>}
 lowerarm_r {parent=upperarm_r, loc=<-26.4793 0 0>, rot=<0 0 -0.2382 -0.9712>}
 hand_r {parent=lowerarm_r, loc=<-22.7458 0 0>, rot=<-0.6747 -0.0182 0.0605 -0.7354>}
 thumb_01_r {parent=hand_r, loc=<-1.3118 0.3191 2.459>, rot=<0.6546 -0.4042 0.0664 -0.6354>}
 thumb_02_r {parent=thumb_01_r, loc=<-5.9322 0.0253 -0.1416>, rot=<0.0121 0.0046 0.2964 -0.955>}
 thumb_03_r {parent=thumb_02_r, loc=<-2.2223 -0.3682 0>, rot=<0 0 0.0172 -0.9999>}
 index_metacarpal_r {parent=hand_r, loc=<-1.6535 -0.0183 2.96>, rot=<0.0158 -0.0124 -0.0173 -0.9996>}
 index_01_r {parent=index_metacarpal_r, loc=<-8.3371 -0.1718 0.2461>, rot=<0.0363 0.0499 0.142 -0.9879>}
 index_02_r {parent=index_01_r, loc=<-3.4919 0.3004 0>, rot=<0 0 0.2948 -0.9556>}
 index_03_r {parent=index_02_r, loc=<-2.5578 0.123 0>, rot=<0 0 0.1968 -0.9804>}
 middle_metacarpal_r {parent=hand_r, loc=<-2.1856 0.1725 0.5984>, rot=<0.006 0.0168 -0.0329 -0.9993>}
 middle_01_r {parent=middle_metacarpal_r, loc=<-7.6374 -0.346 0.1508>, rot=<-0.0085 0.0328 0.1813 -0.9829>}
 middle_02_r {parent=middle_01_r, loc=<-3.9108 0 0>, rot=<0 0 0.2808 -0.9598>}
 middle_03_r {parent=middle_02_r, loc=<-2.6103 -0.0517 0>, rot=<0 0 0.1446 -0.9895>}
 ring_metacarpal_r {parent=hand_r, loc=<-2.2074 0.3625 -1.3638>, rot=<-0.0284 0.0722 -0.0458 -0.9959>}
 ring_01_r {parent=ring_metacarpal_r, loc=<-7.2483 -0.0329 0.1426>, rot=<-0.0456 -0.0277 0.2951 -0.954>}
 ring_02_r {parent=ring_01_r, loc=<-3.7208 -0.4647 0>, rot=<0 0 0.1583 -0.9874>}
 ring_03_r {parent=ring_02_r, loc=<-2.4152 -0.114 0>, rot=<0 0 0.1504 -0.9886>}
 pinky_metacarpal_r {parent=hand_r, loc=<-1.7616 0.9753 -3.3814>, rot=<-0.0397 0.0814 -0.069 -0.9935>}
 pinky_01_r {parent=pinky_metacarpal_r, loc=<-7.1933 0.4205 0.2601>, rot=<-0.0487 -0.0063 0.2556 -0.9655>}
 pinky_02_r {parent=pinky_01_r, loc=<-2.8274 -0.0893 0>, rot=<0 0 0.2474 -0.9689>}
 pinky_03_r {parent=pinky_02_r, loc=<-2.2046 -0.0307 0>, rot=<0 0 0.125 -0.9922>}
 lowerarm_twist_02_r {parent=lowerarm_r, loc=<-9.6199 0 0>, rot=<0 0 0 -1>}
 lowerarm_twist_01_r {parent=lowerarm_r, loc=<-17.8549 0 0>, rot=<0 0 0 -1>}
 upperarm_twist_01_r {parent=upperarm_r, loc=<-0.0041 0 0>, rot=<0 0 0 -1>}
 upperarm_twist_02_r {parent=upperarm_r, loc=<-13.7941 0 0>, rot=<0 0 0 -1>}
 dyn_coat_parent_bk_side_l {parent=pelvis, loc=<-7.1903 -6.4698 -14.0342>, rot=<-0.181 0.2985 0.9368 -0.0245>}
 dyn_coat_01_bk_side_l {parent=dyn_coat_parent_bk_side_l, loc=<0 0 0>, rot=<-0.0023 0.1235 0.0011 -0.9923>}
 dyn_coat_02_bk_side_l {parent=dyn_coat_01_bk_side_l, loc=<21.1749 1.0321 -0.5686>, rot=<0 0 0 1>}
 dyn_coat_03_bk_side_l {parent=dyn_coat_02_bk_side_l, loc=<18.8885 1.3312 -0.3729>, rot=<0 0 0 1>}
 dyn_coat_parent_bk_side_r {parent=pelvis, loc=<-7.0433 -6.5507 14.0187>, rot=<0.0452 -0.9376 -0.2724 0.2114>}
 dyn_coat_01_bk_side_r {parent=dyn_coat_parent_bk_side_r, loc=<0 0 0>, rot=<0 0.1736 0 -0.9848>}
 dyn_coat_02_bk_side_r {parent=dyn_coat_01_bk_side_r, loc=<21.071 -0.4048 -0.3198>, rot=<0 0 0 1>}
 dyn_coat_03_bk_side_r {parent=dyn_coat_02_bk_side_r, loc=<19.4817 -0.8468 -0.7016>, rot=<0 0 0 1>}
 dyn_coat_parent_bk_l {parent=pelvis, loc=<-6.5025 -12.2495 -5.9542>, rot=<-0.1579 0.5779 0.7934 -0.108>}
 dyn_coat_01_bk_l {parent=dyn_coat_parent_bk_l, loc=<0 0 0>, rot=<0 0.1736 0 -0.9848>}
 dyn_coat_02_bk_l {parent=dyn_coat_01_bk_l, loc=<20.7047 -0.4269 -0.3253>, rot=<0 0 0 1>}
 dyn_coat_03_bk_l {parent=dyn_coat_02_bk_l, loc=<19.9635 -1.0802 -0.9864>, rot=<0 0 0 1>}
 dyn_coat_parent_bk_r {parent=pelvis, loc=<-6.3887 -12.0829 5.6688>, rot=<0.0979 -0.796 -0.5797 0.144>}
 dyn_coat_01_bk_r {parent=dyn_coat_parent_bk_r, loc=<0 0 0>, rot=<0.0498 0.1571 0.0177 -0.9862>}
 dyn_coat_02_bk_r {parent=dyn_coat_01_bk_r, loc=<20.5325 0.1834 -0.0769>, rot=<0 0 0 1>}
 dyn_coat_03_bk_r {parent=dyn_coat_02_bk_r, loc=<20.071 0.1062 -1.6851>, rot=<0 0 0 1>}
 thigh_l {parent=pelvis, loc=<-1.0855 -0.2109 -8.7053>, rot=<0.0009 0.0345 0.0155 -0.9993>}
 dyn_coat_01_frnt_l {parent=thigh_l, loc=<-8.1733 9.6756 4.559>, rot=<-0.0255 -0.6149 0.7881 0.0122>}
 dyn_coat_02_frnt_l {parent=dyn_coat_01_frnt_l, loc=<20.3114 -0.0133 0.5007>, rot=<0 0 0 1>}
 dyn_coat_03_frnt_l {parent=dyn_coat_02_frnt_l, loc=<19.1825 0.7405 0.8235>, rot=<0 0 0 1>}
 dyn_coat_01_frnt_side_l {parent=thigh_l, loc=<-7.2004 3.4317 -5.7174>, rot=<-0.0344 -0.3006 0.9531 -0.0083>}
 dyn_coat_02_frnt_side_l {parent=dyn_coat_01_frnt_side_l, loc=<21.4574 0.6838 1.0987>, rot=<0 0 0 1>}
 dyn_coat_03_frnt_side_l {parent=dyn_coat_02_frnt_side_l, loc=<18.5333 0.2392 1.0179>, rot=<0 0 0 1>}
 calf_l {parent=thigh_l, loc=<-42.5268 0 0>, rot=<0 0 -0.0657 -0.9978>}
 foot_l {parent=calf_l, loc=<-46.8037 0 0>, rot=<0.0422 -0.04 0.0613 -0.9964>}
 ball_l {parent=foot_l, loc=<-3.5658 10.8066 -0.1544>, rot=<0.0001 -0.0001 -0.7063 -0.7079>}
 calf_twist_02_l {parent=calf_l, loc=<-18.8064 -0.0967 0.0397>, rot=<0.0237 0.0095 -0.0035 -0.9997>}
 calf_twist_01_l {parent=calf_l, loc=<-35.813 -0.0967 0.0397>, rot=<0.0237 0.0095 -0.0035 -0.9997>}
 thigh_twist_01_l {parent=thigh_l, loc=<-21.629 -0.0919 -0.0002>, rot=<0.0013 0 0 -1>}
 thigh_r {parent=pelvis, loc=<-1.0855 -0.2109 8.7053>, rot=<0.0345 -0.0009 0.9993 0.0155>}
 dyn_coat_01_frnt_side_r {parent=thigh_r, loc=<7.2004 -3.4317 5.7174>, rot=<0.969 -0.0016 0.0539 -0.2413>}
 dyn_coat_02_frnt_side_r {parent=dyn_coat_01_frnt_side_r, loc=<21.3936 -0.9388 1.8775>, rot=<0 0 0 1>}
 dyn_coat_03_frnt_side_r {parent=dyn_coat_02_frnt_side_r, loc=<18.4757 -0.4707 1.7335>, rot=<0 0 0 1>}
 dyn_coat_01_frnt_r {parent=thigh_r, loc=<7.9928 -9.0589 -3.9847>, rot=<0.7433 -0.0194 0.0281 -0.6681>}
 dyn_coat_02_frnt_r {parent=dyn_coat_01_frnt_r, loc=<20.5051 0.4992 1.5031>, rot=<0 0 0 1>}
 dyn_coat_03_frnt_r {parent=dyn_coat_02_frnt_r, loc=<19.4033 -0.4959 1.0844>, rot=<0 0 0 1>}
 calf_r {parent=thigh_r, loc=<42.5268 0 0>, rot=<0 0 -0.0657 -0.9978>}
 foot_r {parent=calf_r, loc=<46.8037 0 0>, rot=<0.0422 -0.04 0.0613 -0.9964>}
 ball_r {parent=foot_r, loc=<3.5658 -10.8066 0.1544>, rot=<0.0001 -0.0001 -0.7063 -0.7079>}
 calf_twist_02_r {parent=calf_r, loc=<18.8064 0.0967 -0.0397>, rot=<0.0237 0.0095 -0.0035 -0.9997>}
 calf_twist_01_r {parent=calf_r, loc=<35.813 0.0967 -0.0397>, rot=<0.0237 0.0095 -0.0035 -0.9997>}
 thigh_twist_01_r {parent=thigh_r, loc=<21.629 0.0919 0.0002>, rot=<0.0013 0 0 -1>}
]
