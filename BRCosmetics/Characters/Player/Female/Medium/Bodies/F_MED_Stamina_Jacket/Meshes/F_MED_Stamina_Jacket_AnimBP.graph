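# Post-process style AnimBP for the female medium Stamina Jacket cosmetic, copying the parent mesh pose and adding coat motion.
# Gameplay chains state-specific Control Rigs for slides, jump, skydive, parachute and crouch, then front-coat calf constraints and pelvis-space RigidBody physics.
# The front end blends to a simpler Control Rig and RigidBody branch.
# EventGraph sets state alphas, picks slide type from left calf angle, adds skydive force and resets dynamics after sliding.
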AnimBlueprint /BRCosmetics/Characters/Player/Female/Medium/Bodies/F_MED_Stamina_Jacket/Meshes/F_MED_Stamina_Jacket_AnimBP
target_skeleton=&/F_MED_Stamina_Jacket_Skeleton
graphs [
 AnimGraph {nodes [
   n1 = "Transform (Modify) Bone - Bone: dyn_coat_01_frnt_r"(ComponentPose=Local To Component(), Alpha=Get Slide_Alpha() -> (Slide_Alpha:double), Translation="0,0,0", Rotation="-100,0,0", Scale="1,1,1") {
    translationMode=BMM_Additive
    translationSpace=BCS_BoneSpace
    alphaBoolBlend {blendOption=Linear}
    boneToModify {boneName=dyn_coat_01_frnt_r}
    rotation {pitch=-100}
   }
   Component To Local(ComponentPose=@n1)
   Component To Local()
   Local To Component()
   n2 = Control Rig(Source=Use cached pose 'CtrlRigCache'()) {
    alphaBoolBlend {blendOption=Linear}
    alphaScaleBiasClamp {interpSpeedDecreasing=4, interpSpeedIncreasing=6}
    controlRigAssetReference {blueprintRigClass=&/F_MED_Stamina_Jacket_FrntEnd_CtrlRig.F_MED_Stamina_Jacket_FrntEnd_CtrlRig_C}
   }
   n3 = RigidBody(ComponentPose=Local To Component(LocalPose=@n2), ExternalForce=Get Velocity_Alpha() -> (Velocity_Alpha:Vector)) {
    bClampLinearTranslationLimitToRefPose=true
    bOverrideWorldGravity=true
    overridePhysicsAsset=&/F_MED_Stamina_Jacket_Physics
    simulationSpace=BaseBoneSpace
    alphaBoolBlend {blendOption=Linear}
    baseBoneRef {boneName=pelvis}
    componentAppliedLinearAccClamp=<3000 3000 3000>
    componentLinearAccScale=<0.15 0.15 0.15>
    componentLinearVelScale=<0.15 0.15 0.15>
    overrideWorldGravity {z=-980}
    simSpaceSettings {
     maxAngularAcceleration=500
     maxAngularVelocity=3
     maxLinearAcceleration=400
     maxLinearVelocity=250
     velocityScaleZ=0
     worldAlpha=1
     externalLinearDragV=<0.1 0.1>
    }
   }
   n4 = Control Rig!(Source=Use cached pose 'CtrlRigCache'(), bAlphaBoolEnabled:bool=Get BaseBall_Slide() -> (BaseBall_Slide:bool)) -> (Pose) {
    alphaInputType=Bool
    alphaBoolBlend {blendOption=Linear}
    controlRigAssetReference {blueprintRigClass=&/F_MED_Stamina_Jacket_BaseballSlide_CtrlRig.F_MED_Stamina_Jacket_BaseballSlide_CtrlRig_C}
   }
   n5 = Control Rig(Source=@n4, bAlphaBoolEnabled=Get Knee_Slide() -> (Knee_Slide:bool)) {
    alphaInputType=Bool
    alphaBoolBlend {blendOption=Linear}
    alphaScaleBiasClamp {interpSpeedIncreasing=1}
    controlRigAssetReference {blueprintRigClass=&/F_MED_Stamina_Jacket_Slide_CtrlRig.F_MED_Stamina_Jacket_Slide_CtrlRig_C}
   }
   n6 = Control Rig!(Source=@n5, Alpha:float=Get Jumping_Alpha() -> (Jumping_Alpha:double)) -> (Pose) {
    alphaBoolBlend {blendOption=Linear}
    alphaScaleBiasClamp {interpSpeedDecreasing=2}
    controlRigAssetReference {blueprintRigClass=&/F_MED_Stamina_Jacket_Jump_CtrlRig.F_MED_Stamina_Jacket_Jump_CtrlRig_C}
   }
   n7 = Control Rig(Source=@n6, Alpha=Get Skydive_Alpha() -> (Skydive_Alpha:double)) {
    alphaBoolBlend {blendOption=Linear}
    controlRigAssetReference {blueprintRigClass=&/F_MED_Stamina_Jacket_Skydive_CtrlRig.F_MED_Stamina_Jacket_Skydive_CtrlRig_C}
   }
   n8 = Control Rig(Source=@n7, Alpha=Get ParachuteOpen_Alpha() -> (ParachuteOpen_Alpha:double)) {
    alphaBoolBlend {blendOption=Linear}
    controlRigAssetReference {blueprintRigClass=&/F_MED_Stamina_Jacket_ParachuteOpen_CtrlRig.F_MED_Stamina_Jacket_ParachuteOpen_CtrlRig_C}
   }
   n9 = Control Rig(Source=@n8, Alpha=Get Crouching_Alpha() -> (Crouching_Alpha:double)) {
    alphaBoolBlend {blendOption=Linear}
    alphaScaleBiasClamp {interpSpeedDecreasing=100, interpSpeedIncreasing=100}
    controlRigAssetReference {blueprintRigClass=&/F_MED_Stamina_Jacket_Crouch_CtrlRig.F_MED_Stamina_Jacket_Crouch_CtrlRig_C}
   }
   n10 = Control Rig(Source=@n9, Alpha=Get Not_CouchOrSkydive_Alpha() -> (Not_CouchOrSkydive_Alpha:double)) {
    alphaBoolBlend {blendOption=Linear}
    alphaScaleBiasClamp {bInterpResult=true, interpSpeedDecreasing=4, interpSpeedIncreasing=6}
    controlRigAssetReference {blueprintRigClass=&/F_MED_Stamina_Jacket_CtrlRig.F_MED_Stamina_Jacket_CtrlRig_C}
   }
   n11 = "Constraint - dyn_coat_02_frnt_l (P)"(ComponentPose=Local To Component(LocalPose=@n10), Alpha=1, ConstraintWeights_0=1) {
    alphaBoolBlend {blendOption=Linear}
    alphaScaleBiasClamp {bClampResult=true, bInterpResult=true, clampMax=0.8, interpSpeedDecreasing=5, interpSpeedIncreasing=7}
    boneToModify {boneName=dyn_coat_02_frnt_l}
    constraintSetup [{targetBone {boneName=calf_l}}]
    constraintWeights [1]
   }
   n12 = "Constraint - dyn_coat_02_frnt_r (P)"(ComponentPose=@n11, Alpha=1, ConstraintWeights_0=1) {
    alphaBoolBlend {blendOption=Linear}
    alphaScaleBiasClamp {bClampResult=true, bInterpResult=true, clampMax=0.8, interpSpeedDecreasing=5, interpSpeedIncreasing=7}
    boneToModify {boneName=dyn_coat_02_frnt_r}
    constraintSetup [{targetBone {boneName=calf_r}}]
    constraintWeights [1]
   }
   n13 = RigidBody(ComponentPose=@n12, ExternalForce=Get Velocity_Alpha()) {
    bClampLinearTranslationLimitToRefPose=true
    bOverrideWorldGravity=true
    overridePhysicsAsset=&F_MED_Stamina_Jacket_Physics
    simulationSpace=BaseBoneSpace
    alphaBoolBlend {blendOption=Linear}
    baseBoneRef {boneName=pelvis}
    componentAppliedLinearAccClamp=<3000 3000 3000>
    componentLinearAccScale=<0.15 0.15 0.15>
    componentLinearVelScale=<0.15 0.15 0.15>
    overrideWorldGravity {z=-980}
    simSpaceSettings {
     maxAngularAcceleration=500
     maxAngularVelocity=3
     maxLinearAcceleration=400
     maxLinearVelocity=250
     velocityScaleZ=0
     worldAlpha=1
     externalLinearDragV=<0.1 0.1>
    }
   }
   Output Pose(Result=Blend Poses by bool(BlendPose_0=Component To Local(ComponentPose=@n3), BlendPose_1=Component To Local(ComponentPose=@n13), bActiveValue=Get bIsInFrontEnd()) {blendType=Linear}) {layerGroup=None}
   Save cached pose 'CtrlRigCache'(Pose=Copy Pose From Mesh() {bUseAttachedParent=true}) {cacheName=CtrlRigCache}
   Get IsSliding_Alpha() -> (IsSliding_Alpha:double)
  ]}
 EventGraph {nodes [
   Event BlueprintInitializeAnimation()
   > Set AnimInstance(AnimInstance:Anim Instance Object Reference=GetAnimInstance(self=GetOwnerSkeletalMeshComponent())) -> (Output_Get:Anim Instance Object Reference)
   Event BlueprintUpdateAnimation()
   > n1 = Sequence!(execute) -> (then_0, then_1, then_2:Exec, then_3:Exec, then_4:Exec, then_5:Exec, then_6:Exec, then_7:Exec, then_8:Exec, then_9:Exec, then_10:Exec)
   > n2 = Branch(Condition=Get bIsCrouching())
   > Set Crouching_Alpha(Crouching_Alpha:double=1) -> (Output_Get:double)
   Set Crouching_Alpha(execute=@n2.else, Crouching_Alpha=0)
   Set IsInFrontend(execute=@n1.then_1, IsInFrontend:double=SelectFloat(A=1, B=0, bPickA=Get bIsInFrontEnd())) -> (Output_Get:double)
   n3 = Get bIsSkydiveFloating()
   n4 = Get bIsSkydiveDiveMode()
   Set Velocity_Alpha(execute=@n1.then_4, Velocity_Alpha:Vector=SelectVector(A="0, 0,25000", B="0, 0, 0", bPickA=OR Boolean(A=@n4, B=@n3))) -> (Output_Get:Vector)
   n5 = NOT Boolean(A=Get bIsParachuteOpened())
   n6 = OR Boolean(A=Get bIsJumping(), B=Get bIsFalling())
   Set Jumping_Alpha(execute=@n1.then_5, Jumping_Alpha:double=SelectFloat(A=1, B=0, bPickA=AND Boolean(A=@n6, B=@n5))) -> (Output_Get:double)
   n7 = Branch(execute=@n1.then_8, Condition=Get bIsSliding())
   > n8 = Set Left_Calf(Left_Calf:double=GetAlphaFromReferenceTransformDelta(SkeletalMeshComponent=GetPartSkeletalMeshComponent(), BoneName=calf_l, BaseName=thigh_l, InputRange=(X=0,Y=1), TwistAxis="1,0,0").RawOutput) -> (Output_Get:double)
   > n9 = Branch(Condition=NOT Boolean(A="float > float"(A=@n8, B=100)))
   > Set BaseBall_Slide(BaseBall_Slide:bool=true) -> (Output_Get:bool)
   > Set Knee_Slide(Knee_Slide:bool=false) -> (Output_Get:bool)
   Set Knee_Slide(execute=@n9.else, Knee_Slide=true)
   > Set BaseBall_Slide(BaseBall_Slide=false)
   Set BaseBall_Slide(execute=@n7.else, BaseBall_Slide=false)
   > Set Knee_Slide(Knee_Slide=false)
   Set Skydive_Alpha(execute=@n1.then_2, Skydive_Alpha:double=SelectFloat(A=1, B=0, bPickA=Get bIsSkyDiving())) -> (Output_Get:double)
   Branch(execute=@n1.then_10, Condition=Get From_Slide() -> (From_Slide:bool))
   > Branch(Condition=NOT Boolean(A=Get bIsSliding()))
   > Delay(Duration=0.15)
   > ResetDynamics(InTeleportType=TeleportPhysics)
   > ResetDynamics(InTeleportType=ResetPhysics)
   > n10 = Set From_Slide(From_Slide:bool=false) -> (Output_Get:bool)
   Do Once(execute=@n12.then, Reset=@n10.then)
   > Set From_Slide(From_Slide=true)
   n11 = OR Boolean!(A=Get bIsSkyDiving(), B=Get bIsCrouching(), C:bool=Get bIsJumping()) -> (ReturnValue)
   Set Not_CouchOrSkydive_Alpha(execute=@n1.then_3, Not_CouchOrSkydive_Alpha:double=SelectFloat(A=1, B=0, bPickA=NOT Boolean(A=@n11))) -> (Output_Get:double)
   Set ParachuteOpen_Alpha(execute=@n1.then_6, ParachuteOpen_Alpha:double=SelectFloat(A=1, B=0, bPickA=Get bIsParachuteOpened())) -> (Output_Get:double)
   n12 = Branch(execute=@n1.then_9, Condition=Get bIsSliding())
   Branch(Condition=Get bIsSliding())
   n13 = GetPartSkeletalMeshComponent()
   > n14 = Cast To SlidingLayerAnimBP!(
    execute
    Object=GetLinkedAnimLayerInstanceByClass!(
     self:Anim Instance Object Reference=GetAnimInstance(self=@n13)
     InClass:Anim Instance Class Reference=/Game/Characters/Player/Male/Male_Avg_Base/LayerAnimBP/SlidingLayerAnimBP.SlidingLayerAnimBP_C
     bCheckForChildClass:bool=false
    ) -> (ReturnValue:Anim Instance Object Reference)
   ) -> (then, CastFailed, AsSliding Layer Anim BP:Sliding Layer Anim BP Object Reference) {is=title K2Node_DynamicCast}
   > n15 = Branch(Condition=Get bIsKneeSliding!(self:Fort Sliding Layer Anim Instance Object Reference=@n14) -> (bIsKneeSliding))
   > Set Slide_Alpha(Slide_Alpha:double=0) -> (Output_Get:double)
   Set Slide_Alpha(execute=@n15.else, Slide_Alpha=1)
   PrintString!(execute=@n14.CastFailed, InString=Failed casting to SlidingLayerAnimBP, bPrintToScreen, bPrintToLog=false, TextColor, Duration, Key) -> (then)
   MapRangeClamped(
    Value=GetTwistAndSwingAngleOfDeltaRotationFromRefPose(self=GetOwnerSkeletalMeshComponent(), BoneName=calf_l).OutSwingAngle
    InRangeA=4
    InRangeB=150
   )
   FInterpTo(Current=Get Crouching_Alpha(), Target=0, DeltaTime=Reroute Node(InputPin:double) -> (OutputPin:double), InterpSpeed=0.5)
   n16 = Get PawnSpeed2D()
   SelectFloat(A=1, B=0, bPickA=AND Boolean(A=Get bIsCrouching(), B="float <= float"(A=@n16, B=1) -> (ReturnValue=~)))
   FInterpTo(Current=Get Crouching_Alpha(), Target=1, DeltaTime=Reroute Node(), InterpSpeed=0.5)
   OR Boolean!(A=Get bIsSkydiveFloating(), B=Get bIsSkydiveDiveMode()) -> (ReturnValue)
   Set IsSliding_Alpha(IsSliding_Alpha:double=SelectFloat(A=1, B=0, bPickA=Get bIsSliding())) -> (Output_Get:double)
  ]}
]
variables [
 AnimInstance {type=AnimInstance, default=null}
 BaseBall_Slide {type=boolean}
 Crouching_Alpha {type=number}
 From_Slide {type=boolean}
 IsInFrontend {type=number}
 IsSliding_Alpha {type=number}
 Jumping_Alpha {type=number}
 Knee_Slide {type=boolean}
 Left_Calf {type=number}
 legSwingDelta_r {type=number}
 Not_CouchOrSkydive_Alpha {type=number}
 ParachuteOpen_Alpha {type=number}
 Skel_Mesh {type=SkeletalMeshComponent}
 Skydive_Alpha {type=number}
 Slide_Alpha {type=number}
 TestVar {type=number}
 Velocity_Alpha {type=Vector}
]
